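# Lit material for the power cord and plug Niagara mesh particles in the FruitFire Nudge skydiving contrail.
# Vertex color and a UV1 linear gradient form an inverted, power-sharpened mask along the mesh.
# That mask offsets a time-driven sine wave and weights both wave height and droop, with dynamic parameters controlling speed, height, droop and sine scale.
# A near-black color constant, AO texture and normal map supply the surface look.
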Material /BRCosmetics/Athena/Effects/Particles/Skydiving/NS_Contrail_FruitFire_Nudge/Materials/M_Contrail_FruitFire_Nudge_CordandPlug
graph {
 properties {
  bAutomaticallySetUsageInEditor=false
  bUsedWithNiagaraMeshParticles=true
  previewMesh=/BRCosmetics/Athena/Effects/Particles/Skydiving/NS_Contrail_FruitFire_Nudge/Mesh/SM_Contrail_FruitFire_Nudge_TVmesh_PowerCord
  usedShadingModels=Default Lit
 }
 nodes [
  n1 = DynamicParameter() {parameterIndex=3, defaultValue=<rgba -1 10 0 1>, paramNames [WpoPanSpeed,WpoHeight,WpoDroop,SineScale]}
  n2 = Multiply(A=Time(), B=@n1.WpoPanSpeed)
  n3 = VertexColor()
  n4 = TextureSample(UVs=TextureCoordinate() {coordinateIndex=1}) {
   constCoordinate=1
   samplerSource=SSM_Clamp_WorldGroupSettings
   texture=/BRCosmetics/Athena/Effects/Particles/Skydiving/NS_Contrail_FruitFire_Nudge/Materials/Textures/LinearGradient
  }
  n5 = Power(Base=OneMinus(None=Multiply(A=@n3, B=@n4.RGB))) {constExponent=4}
  n6 = Sine(None=Multiply(A=Add(A=@n2, B=@n5), B=@n1.SineScale) {constB=0.5})
  n7 = Clamp(None=@n5)
  Add(A=Multiply(A=Multiply(A=@n7, B=@n6), B=@n1.WpoHeight), B=Multiply(A=@n5, B=@n1.WpoDroop) {constB=-55})
  Constant3Vector() {constant=<rgba 0.026 0.026 0.026 0>}
  Constant()
  TextureSample() {texture=/BRCosmetics/Athena/Effects/Particles/Skydiving/NS_Contrail_FruitFire_Nudge/Materials/Textures/FruitFire_Nudge_ao}
  TextureSample() {
   samplerType=SAMPLERTYPE_Normal
   texture=/BRCosmetics/Athena/Effects/Particles/Skydiving/NS_Contrail_FruitFire_Nudge/Materials/Textures/FruitFire_Nudge_normal
  }
  VertexNormalWS()
 ]
}
