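# Lit mesh-particle material for the FruitFire Nudge skydiving contrail, showing TV screen textures.
# A DynamicParameter offsets the half-scale atlas UVs of the first texture, using FlipbookU on both axes, and TextureSwap blends to a second, full-UV Gracie texture.
# The blend feeds a saturation boost, an emissive adjustment function scaled by EmissiveIntensity, and a lerp masked to a horizontal band around the UV centre that widens with EmissiveIntensity.
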Material /BRCosmetics/Athena/Effects/Particles/Skydiving/NS_Contrail_FruitFire_Nudge/Materials/M_Contrail_FruitFire_Nudge_Textures
graph {
 properties {bAutomaticallySetUsageInEditor=false, bUsedWithNiagaraMeshParticles=true, usedShadingModels=Default Lit}
 nodes [
  Constant()
  n1 = TextureCoordinate() {vTiling=0.85}
  n2 = DynamicParameter() {parameterIndex=2, defaultValue=<rgba 0.5 0 0 1>, paramNames [FlipbookU, FlipbookV, EmissiveIntensity, TextureSwap]}
  n3 = TextureSampleParameter2D(UVs=Add(A=Multiply(A=@n1) {constB=0.5}, B=AppendVector(A=@n2.FlipbookU, B=@n2.FlipbookU))) {
   parameterName=TVScreens
   texture=/BRCosmetics/Athena/Effects/Particles/Skydiving/NS_Contrail_FruitFire_Nudge/Materials/Textures/T_Contrail_FruitFire_TVTextures
  }
  n4 = TextureSampleParameter2D() {
   parameterName=TVScreens2
   texture=/BRCosmetics/Athena/Effects/Particles/Skydiving/NS_Contrail_FruitFire_Nudge/Materials/Textures/T_Contrail_FruitFire_TVTextures_Gracie
  }
  n5 = LinearInterpolate(A=@n3.RGB, B=@n4.RGB, Alpha=@n2.TextureSwap)
  Desaturation(None=@n5, Fraction=Constant() {r=-0.2})
  n6 = Subtract(A=ComponentMask(None=@n1) {g=true}) {constB=0.5}
  LinearInterpolate(A=Constant(), B=@n5, Alpha=Step(Y=Abs(None=@n6), X=@n2.EmissiveIntensity) {constX=0.01})
  MaterialFunctionCall(Emissive=Multiply(A=@n5, B=@n2.EmissiveIntensity) {constB=0.12}) {materialFunction=&MF_FX_EmissiveAdjustments}
 ]
}
