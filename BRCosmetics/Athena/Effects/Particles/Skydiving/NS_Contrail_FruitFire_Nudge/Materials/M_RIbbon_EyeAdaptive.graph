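# Unlit additive Niagara ribbon and sprite material in the FruitFire Nudge skydiving contrail effect folder.
# Particle color drives emissive through MF_FX_EmissiveAdjustments, while particle alpha scales a reversible U-gradient mask.
# An optional panning Voronoi noise texture, with optional contrast, modulates that mask, using dynamic parameters for Y tiling and X pan speed.
# The default-on camera fade switch multiplies in distance fade plus an inverted U gradient.
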Material /BRCosmetics/Athena/Effects/Particles/Skydiving/NS_Contrail_FruitFire_Nudge/Materials/M_RIbbon_EyeAdaptive
graph {
 properties {
  bEnableResponsiveAA=true
  bOutputTranslucentVelocity=true
  bUsedWithNiagaraRibbons=true
  bUsedWithNiagaraSprites=true
  bUsedWithParticleSprites=true
  blendMode=BLEND_Additive
  pixelDepthOffsetMode=PDOM_Legacy
  shadingModel=MSM_Unlit
  usedShadingModels=Unlit
 }
 nodes [
  n1 = ParticleColor()
  Reroute(None=MaterialFunctionCall(Emissive=@n1.RGB) {materialFunction=&MF_FX_EmissiveAdjustments}.Result)
  ScalarParameter() {defaultValue=1, parameterName=Tex_Y}
  ScalarParameter() {parameterName=Pan_X}
  n2 = TextureCoordinate()
  n3 = DynamicParameter() {defaultValue=<rgba 1 1 -0.7 1>, paramNames [Param1,Y,PanX_Speed,Param4]}
  n4 = AppendVector(A=ScalarParameter() {defaultValue=1, parameterName=Tex_X}, B=@n3.Y)
  n5 = ScalarParameter() {parameterName=Pan_Y}
  n6 = TextureSampleParameter2D(UVs=Panner(Coordinate=Multiply(A=@n2, B=@n4), Speed=AppendVector(A=@n3.PanX_Speed, B=@n5))) {parameterName=Texture, texture=/BRCosmetics/Athena/Effects/Particles/Skydiving/NS_Contrail_FruitFire_Nudge/Materials/Textures/T_VoronoiNoise}
  n7 = ComponentMask(None=TextureCoordinate() {uTiling=0.5}) {r=true}
  n8 = Power(
   Base=Multiply(A=@n7, B=ScalarParameter() {defaultValue=1, parameterName=MaskMult})
   Exp=ScalarParameter() {defaultValue=1, parameterName=MaskPow}
  )
  n9 = StaticSwitchParameter(True=@n8, False=OneMinus(None=@n8)) {parameterName=Reverse Mask?}
  n10 = Multiply(A=Multiply(A=@n6.R, B=ScalarParameter() {defaultValue=1, parameterName=Tex_Mult}), B=@n9)
  n11 = MaterialFunctionCall(In=@n10, Contrast=ScalarParameter() {parameterName=CheapContrast}) {materialFunction=&CheapContrast}
  n12 = StaticSwitchParameter(
   True=Multiply(A=@n1.A, B=StaticSwitchParameter(True=@n11.Result, False=@n10) {parameterName=Use Cheap Contrast})
   False=Multiply(A=@n1.A, B=@n9)
  ) {parameterName=Use Texture?}
  n13 = OneMinus(None=@n7)
  n14 = MaterialFunctionCall!(Distance=ScalarParameter() {defaultValue=250, parameterName=Camera Fade}) {materialFunction=/Game/Effects/Fort_Effects/Materials/_Functions/MF_CameraDistanceFade}
  StaticSwitchParameter(True=Multiply(A=@n12, B=Multiply(A=@n13, B=@n14.Result)), False=@n12) {defaultValue=true, parameterName=Use CamFade?}
 ]
}
