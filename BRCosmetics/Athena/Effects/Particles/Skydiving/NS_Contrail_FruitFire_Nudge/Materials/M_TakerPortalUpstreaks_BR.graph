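# Translucent, unlit, two-sided Niagara sprite and mesh material for streak effects in a skydiving contrail, colored by particle color through an emissive adjustment function.
# Two stretched textures, caustics and noise, pan at Dynamic Parameter speeds and combine into streaks.
# A V-gradient mask with optional Fresnel shapes them, then particle alpha and CheapContrast finish the opacity.
# The same mask and alpha drive a slight refraction value, and the noise also builds a normal-style vector.
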Material /BRCosmetics/Athena/Effects/Particles/Skydiving/NS_Contrail_FruitFire_Nudge/Materials/M_TakerPortalUpstreaks_BR
graph {
 properties {
  bUsedWithNiagaraMeshParticles=true
  bUsedWithNiagaraSprites=true
  blendMode=BLEND_Translucent
  pixelDepthOffsetMode=PDOM_Legacy
  refractionMethod=RM_IndexOfRefraction
  shadingModel=MSM_Unlit
  twoSided=true
  usedShadingModels=Unlit
 }
 nodes [
  n1 = ParticleColor()
  n2 = Power(Base=ComponentMask(None=TextureCoordinate()) {g=true})
  n3 = MaterialFunctionCall(Exponent=ScalarParameter() {defaultValue=4, parameterName=FresnelExponent}) {materialFunction=&MF_Fresnel_FX}
  n4 = ScalarParameter() {defaultValue=1, parameterName=UseFresnel}
  n5 = TextureCoordinate()
  n6 = Multiply(A=OneMinus(None=ComponentMask(None=@n5) {g=true})) {constB=5}
  n7 = Multiply(A=Multiply(A=@n2, B=LinearInterpolate(B=@n3.Fresnel, Alpha=@n4) {constA=1}), B=Power(Base=Saturate(None=@n6)) {constExponent=1})
  LinearInterpolate(Alpha=Multiply(A=@n1.A, B=@n7)) {constA=1, constB=1.1}
  n8 = Constant()
  n9 = DynamicParameter() {defaultValue=<rgba 0.5 0.35 1 1>, paramNames [PanY1,PanY2,Param3,Param4]}
  n10 = TextureSample(UVs=Panner(Coordinate=TextureCoordinate() {uTiling=0.8, vTiling=0.1}, Speed=AppendVector(A=@n8, B=@n9.PanY2)) {speedY=0.35}) {texture=&T_NoisePack_01}
  n11 = MaterialFunctionCall(Float2=ConstantBiasScale(None=MaterialFunctionCall(X=@n10.G, Y=@n10.G) {materialFunction=&MakeFloat2}.Result) {bias=-0.5, scale=2}) {materialFunction=&BreakOutFloat2Components}
  MaterialFunctionCall(X=@n11.R, Y=@n11.R, Z=Constant() {r=1}) {materialFunction=&MakeFloat3}
  MaterialFunctionCall(Emissive=@n1.RGB) {materialFunction=&MF_FX_EmissiveAdjustments}
  n12 = Panner(Coordinate=TextureCoordinate() {uTiling=0.6, vTiling=0.01}, Speed=AppendVector(A=Constant(), B=@n9.PanY1)) {speedY=0.5}
  n13 = Saturate(None=Add(A=TextureSample(UVs=@n12) {texture=&T_ShieldCaustics}.R, B=Multiply(A=@n10.R, B=@n2) {constB=0.8}))
  MaterialFunctionCall(In=Multiply(A=@n1.A, B=Multiply(A=@n13, B=@n7)), Contrast=ScalarParameter() {defaultValue=0.01, parameterName=CheapContrast}) {materialFunction=&CheapContrast}
 ]
}
