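# Niagara skydiving contrail cosmetic that surrounds the player with tumbling TV meshes, cloud puffs, stars and speedlines.
# TVs_and_Clouds drives the meshes; SpawningClouds and TV_Speedlines read its particles to add bursts at spawn and following streaks.
# Ambient_Speedlines spreads sprites across four skeletal mesh bones or sockets.
# FN_Contrail_Setup supplies direction and intro fade, while front-end modifiers and time-of-day functions adjust scale, rates, motion and color.
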NiagaraSystem /BRCosmetics/Athena/Effects/Particles/Skydiving/NS_Contrail_FruitFire_Nudge/NS_Contrail_FruitFire_Nudge
dependencies {
 usedDataInterfaces [
  &NiagaraDataInterfaceVectorCurve
  &NiagaraDataInterfaceCurve
  &NiagaraDataInterfaceArrayFloat
  &NiagaraDataInterfaceParticleRead
  &NiagaraDataInterfaceSpriteRendererInfo
  &NiagaraDataInterfaceSkeletalMesh
 ]
 usedDynamicInputs [&MakeQuatFromAxisAngle, &Lerp_Float]
 usedModules [
  &self:SystemSpawnScript.SystemScriptSource.SystemScriptGraph.NiagaraNodeAssignment_0.SetVariables_#1
  &SystemState
  /Game/Effects/Niagara/Modules/Contrails/FN_Contrail_Setup
  &FN_Integrator
  &self:SystemSpawnScript.SystemScriptSource.SystemScriptGraph.NiagaraNodeAssignment_1.SetVariables_#2
 ]
}
emitters {
 Ambient_Speedlines {
  emitterUpdateScript [
   &EmitterState()
   &SpawnRate(SpawnRate=15, Use Spawn Probability=false)
  ]
  particleSpawnScript [
   &PartitionParticles(Number Of Partitions=4)
   &InitializeParticle(
    Color=<rgba 1 1 1 0.5>
    Lifetime Max=1
    Lifetime Min=0.5
    Lifetime Mode=::NewEnumerator1
    Sprite Size Max=<0.6 75>
    Sprite Size Min=<0.75 10>
    Sprite Size Mode=::NewEnumerator2
    UsePositionOffset=false
    Write Material Random=true
   )
   &ShapeLocation(
    Apply To Particle Position=::NewEnumerator1
    MaskSpawnGroup=false
    Override Seed=false
    Shape Origin=&VectorFromFloat(Value=0)
    Sphere Radius=75
    UseOwnerScale=true
   )
  ]
  particleUpdateScript [
   &ParticleState()
   &SkeletalMeshLocation(
    Bone / Socket Index=$Particles.PartitionParticles.Partition
    MaskSpawnGroup=false
    Mesh Sampling Type=::NewEnumerator5
    Orientation Sampling=::NewEnumerator1
    Sampled Position Offset=$Particles.ShapeLocation.ShapePosition
    Skeletal Mesh=$System.Skeletal Mesh
    Socket Sampling Mode=::NewEnumerator5
   )
   /SetVariables(Particles.NewAge:NiagaraFloat=&Multiply_Float(A=$Particles.Age, B=0.75)) {isSetParameters=true}
   /SetVariables(Particles.Velocity=&Multiply_VectorByFloat(Vector=&Lerp_Vector(A=<0 0 0>, B=$System.FN_Contrail_Setup.Direction Velocity, Alpha=$Particles.NewAge), Float=1)) {isSetParameters=true}
   /SetVariables(Particles.Position=&AddVectorToPosition(
     Position=$Particles.Position
     Vector=&Multiply_VectorByFloat(Vector=$Particles.Velocity, Float=$Particles.Age)
    )) {isSetParameters=true}
   /SetVariables(Particles.Velocity=$System.FN_Contrail_Setup.Direction Vector) {isSetParameters=true}
   &ScaleColor(
    Scale Alpha=&FloatFromCurve(
     FloatCurve=DataInterface() {propertyValues {Curve {keys [
         {interpMode=RCIM_Cubic, tangentMode=RCTM_User, time=0, value=0}
         {interpMode=RCIM_Cubic, tangentMode=RCTM_User, time=0.5, value=1}
         {tangentMode=RCTM_User, time=1, value=0}
        ]}}}
     Scale Curve=$System.FN_Contrail_Setup.Normalized Intro Age
    )
    Scale RGB=&MakeVectorFromLinearColorRGB(Color=$System.ToD)
    ScaleA=true
    ScaleRGB=true
   )
   /SetVariables(
    Particles.SpriteOutline:LinearColor=&MultiplyLinearColor(Linear Color=$Particles.Color, Scale Factor=&MakeVector4_FromLinearColor(LinearColor=<rgba 0 0 0 1>))
    Particles.SpriteOutlineWidth:Vector2f=&Multiply_Vector2D(A=$Particles.SpriteSize, B=&Multiply_Vector2D_ByFloat(Float=1.25, Vector2D=<1 1>))
   ) {isSetParameters=true}
  ]
  rendererClasses [&NiagaraSpriteRendererProperties]
  renderers [
   {rendererClass=&NiagaraSpriteRendererProperties}
   {rendererClass=&NiagaraSpriteRendererProperties, rendererIndex=1}
  ]
 }
 SpawningClouds {
  emitterUpdateScript [
   &EmitterState(Life Cycle Mode=::NewEnumerator1, Loop Duration=0.1, Recalculate Duration Each Loop=true, UseLoopDelay=false)
   &SpawnParticlesFromOtherEmitter(
    Attribute Reader=DataInterface() {propertyValues {EmitterBinding {bindingMode=Other, emitterName=TVs_and_Clouds}}}
    Maximum Age Spawn Threshold=0.3
    Minimum Age Spawn Threshold=0.001
    Spawn Rate=&RandomRangeFloat(Minimum=80, Maximum=90)
    Use Maximum Age=false
    Use Minimum Age=false
   )
  ]
  particleSpawnScript [
   &InitializeParticle(
    Color Mode=::NewEnumerator0
    Lifetime Max=1
    Lifetime Min=0.7
    Lifetime Mode=::NewEnumerator1
    Mesh Scale Mode=::NewEnumerator3
    Mesh Uniform Scale=&Multiply_Float(
     A=&RandomRangeFloat(Minimum=1.5, Maximum=2)
     B=&FN_FrontEndFloatModifier(Gameplay=1, Front End=0.4)
    )
    Sprite Rotation Mode=::NewEnumerator2
    Sprite Size Mode=::NewEnumerator3
    Uniform Sprite Size=&Multiply_Float(
     A=&RandomRangeFloat(Minimum=30, Maximum=40)
     B=&FN_FrontEndFloatModifier(Gameplay=1, Front End=0.4)
    )
    UsePositionOffset=false
    Write Material Random=true
   )
   &SampleParticlesFromOtherEmitter(
    Lifetime Application:ENiagara_LifetimeInheritanceOptions=::NewEnumerator1
    Lifetime Sampling:ENiagara_DirectReadSamplingMode=::NewEnumerator0
    Maximum Sampled Particle Age=0.05
    Mesh Scale Sampling:ENiagara_DirectReadSamplingMode=::NewEnumerator1
    Minimum Sampled Particle Age=0.01
    Module.MaxSampledAgeCutoff=true
    Module.MinSampledAgeCutoff=true
    Particle ID Sampling=::NewEnumerator1
    Reject Particle By Sampled Age=true
    Reject Particle on Sampled Particle's First Frame=true
   )
   &ShapeLocation(MaskSpawnGroup=false, Override Seed=false, Sphere Radius=85, UseOwnerScale=true)
   &ApplyInitialForces()
   &InitialMeshOrientation(Use Rotation Vector=false)
   &SubUVAnimation(
    Evaluation Type=::NewEnumerator0
    "Sprite Renderer "=DataInterface() {propertyValues {SpriteRenderer=&self:SpawningClouds_0.NiagaraSpriteRendererProperties_0}}
    SubUV Animation Mode:ENiagara_SubUVLookupModeV2=::NewEnumerator2
    UseEndFrame=false
    UseStartFrame=false
   )
  ]
  particleUpdateScript [
   &ParticleState()
   /SetVariables(
    Particles.CloudScale:Vector3f=&VectorFromCurve(
     VectorCurve=DataInterface() {propertyValues {
       XCurve {keys [
         {arriveTangent=34.4414, interpMode=RCIM_Cubic, leaveTangent=34.4414, tangentMode=RCTM_User, time=0, value=0}
         {arriveTangent=5.0132, interpMode=RCIM_Cubic, leaveTangent=5.0132, time=0.0834, value=1.9435}
         {arriveTangent=-4.6649, interpMode=RCIM_Cubic, leaveTangent=-4.6649, time=0.3534, value=1.7719}
         {arriveTangent=-11.0402, tangentMode=RCTM_User, time=0.5, value=0}
        ]}
       YCurve {keys [
         {arriveTangent=34.4414, interpMode=RCIM_Cubic, leaveTangent=34.4414, tangentMode=RCTM_User, time=0, value=0}
         {arriveTangent=5.0132, interpMode=RCIM_Cubic, leaveTangent=5.0132, time=0.0834, value=1.9435}
         {arriveTangent=-4.6649, interpMode=RCIM_Cubic, leaveTangent=-4.6649, time=0.3534, value=1.7719}
         {arriveTangent=-14.228, leaveTangent=-3.8596, tangentMode=RCTM_User, time=0.5, value=0}
        ]}
       ZCurve {keys [
         {arriveTangent=34.4414, interpMode=RCIM_Cubic, leaveTangent=34.4414, tangentMode=RCTM_User, time=0, value=0}
         {arriveTangent=5.0132, interpMode=RCIM_Cubic, leaveTangent=5.0132, time=0.0834, value=1.9435}
         {arriveTangent=-4.6649, interpMode=RCIM_Cubic, leaveTangent=-4.6649, time=0.3534, value=1.7719}
         {arriveTangent=-13.7325, tangentMode=RCTM_User, time=0.5, value=0}
        ]}
      }}
     Scale Curve=&Multiply_Vector(
      A=&VectorFromFloat(Value=&RandomRangeFloat(Minimum=1.3, Maximum=1.5))
      B=&FN_FrontEndVector3Modifier(Gameplay=<1 1 1>, Front End=<0.4 0.4 0.4>)
     )
    )
    Particles.Cloud_DMP:Vector4f=&MakeVector4(
     X=&FloatFromCurve(FloatCurve=DataInterface() {propertyValues {Curve {
         defaultValue=0
         keys [
          {arriveTangent=1.2525, interpMode=RCIM_Cubic, leaveTangent=1.2525, tangentMode=RCTM_User, time=0, value=0}
          {arriveTangent=1.5209, interpMode=RCIM_Cubic, leaveTangent=1.5209, tangentMode=RCTM_User, time=0.3, value=1}
         ]
        }}})
     Y=1
     Z=1
     W=1
    )
    Particles.StarColor:LinearColor=&Multiply_LinearColorByFloat(
     Linear Color=<rgba 1 0.602 0.1667 1>
     Float=&FloatFromCurve(FloatCurve=DataInterface() {propertyValues {Curve {
         defaultValue=0
         keys [
          {tangentMode=RCTM_User, time=0, value=1}
          {arriveTangent=-0.0405, interpMode=RCIM_Cubic, leaveTangent=-0.0405, tangentMode=RCTM_User, time=0.7009, value=1}
          {arriveTangent=-8.0089, interpMode=RCIM_Cubic, leaveTangent=-8.0089, tangentMode=RCTM_User, time=1, value=0}
         ]
        }}})
    )
    Particles.StarColor_Outline:LinearColor=&Multiply_LinearColorByFloat(
     Linear Color=<rgba 0 0 0 1>
     Float=&FloatFromCurve(FloatCurve=DataInterface() {propertyValues {Curve {
         defaultValue=0
         keys [
          {tangentMode=RCTM_User, time=0, value=1}
          {arriveTangent=-0.0405, interpMode=RCIM_Cubic, leaveTangent=-0.0405, tangentMode=RCTM_User, time=0.7009, value=1}
          {arriveTangent=-8.0089, interpMode=RCIM_Cubic, leaveTangent=-8.0089, tangentMode=RCTM_User, time=1, value=0}
         ]
        }}})
    )
    Particles.StarPosition:NiagaraPosition=&Lerp_Vector(
     A=Linked() {linkedVariable {name=Particles.Position, type=NiagaraPosition}}
     B=&AddVectorToPosition(Position=$Particles.Position, Vector=&RandomRangeVector(Minimum=<-250 -250 0>, Maximum=<250 250 250>))
     Alpha=&FloatFromCurve(FloatCurve=DataInterface() {propertyValues {Curve {
         defaultValue=1
         keys [
          {arriveTangent=4.1785, interpMode=RCIM_Cubic, leaveTangent=4.1785, tangentMode=RCTM_User, time=0, value=0}
          {arriveTangent=1.1179, interpMode=RCIM_Cubic, leaveTangent=1.1179, tangentMode=RCTM_User, time=0.2042, value=0.6228}
          {arriveTangent=0.0224, interpMode=RCIM_Cubic, leaveTangent=0.0224, tangentMode=RCTM_User, time=1, value=1}
         ]
        }}})
    )
   ) {isSetParameters=true}
   /Niagara/Modules/AttributeReader/UpdateParticlesFromOtherEmitter(
    Attribute Reader=DataInterface() {propertyValues {EmitterBinding {bindingMode=Other, emitterName=TVs_and_Clouds}}}
    Position Sampling:ENiagara_DirectReadSamplingMode=::NewEnumerator3
   )
   &PointForce(
    Force Origin:NiagaraPosition=$StackContext.ShapeLocation.ShapePosition
    Force Strength=&FloatFromCurve(
     FloatCurve=DataInterface() {propertyValues {Curve {
        defaultValue=0
        keys [
         {arriveTangent=-2.5249, interpMode=RCIM_Cubic, leaveTangent=-2.5249, tangentMode=RCTM_User, time=0, value=3}
         {arriveTangent=-1.5724, interpMode=RCIM_Cubic, leaveTangent=-1.5724, tangentMode=RCTM_User, time=0.4985, value=0.3106}
         {arriveTangent=-0.0405, interpMode=RCIM_Cubic, leaveTangent=-0.0405, tangentMode=RCTM_User, time=1, value=0}
        ]
       }}}
     Scale Curve=&RandomRangeFloat(Minimum=600, Maximum=800)
    )
    Linear Force Vector:Vector3f=&RandomRangeVector(Minimum=<-1 -1 0.1>, Maximum=<1 1 1>)
    Module.Use Falloff=false
    Module.UseLinearForce=true
   )
   &Drag(
    Drag=&FloatFromCurve(
     FloatCurve=DataInterface() {propertyValues {Curve {
        defaultValue=0
        keys [
         {arriveTangent=9.0725, interpMode=RCIM_Cubic, leaveTangent=9.0725, tangentMode=RCTM_User, time=0, value=0}
         {arriveTangent=-2.5249, interpMode=RCIM_Cubic, leaveTangent=-2.5249, tangentMode=RCTM_User, time=0.1408, value=0.9847}
         {arriveTangent=-0.0405, interpMode=RCIM_Cubic, leaveTangent=-0.0405, tangentMode=RCTM_User, time=1, value=0}
        ]
       }}}
     Scale Curve=&RandomRangeFloat(Minimum=17, Maximum=25)
    )
    Use Linear Drag=true
    Use Rotational Drag=false
   )
   &GravityForce(Gravity=<0 0 900>)
   &SolveForcesAndVelocity(Clamp Velocity=false, Limit Acceleration=false)
   &ScaleMeshSize(Scale Factor=&VectorFromFloat(Value=&FloatFromCurve(FloatCurve=DataInterface() {propertyValues {Curve {
         defaultValue=1
         keys [
          {arriveTangent=4.1785, interpMode=RCIM_Cubic, leaveTangent=4.1785, tangentMode=RCTM_User, time=0, value=0}
          {arriveTangent=1.1179, interpMode=RCIM_Cubic, leaveTangent=1.1179, tangentMode=RCTM_User, time=0.2042, value=0.6228}
          {arriveTangent=0.0224, interpMode=RCIM_Cubic, leaveTangent=0.0224, tangentMode=RCTM_User, time=1, value=1}
         ]
        }}})))
   &SpriteRotationRate(Rotation Rate=&RandomRangeFloat(Minimum=-100, Maximum=100))
   &ScaleSpriteSize(Uniform Curve Sprite Scale=DataInterface() {propertyValues {Curve {
       defaultValue=0
       keys [
        {arriveTangent=30.3264, interpMode=RCIM_Cubic, leaveTangent=30.3264, tangentMode=RCTM_User, time=0, value=0}
        {arriveTangent=-2.5249, interpMode=RCIM_Cubic, leaveTangent=-2.5249, tangentMode=RCTM_User, time=0.1408, value=1.2}
        {arriveTangent=-0.7591, interpMode=RCIM_Cubic, leaveTangent=-0.7591, tangentMode=RCTM_User, time=0.2108, value=0.385}
        {arriveTangent=-0.9894, interpMode=RCIM_Cubic, leaveTangent=-0.9894, time=0.3032, value=0.7683}
        {arriveTangent=-0.0405, interpMode=RCIM_Cubic, leaveTangent=-0.0405, tangentMode=RCTM_User, time=0.6, value=0}
       ]
      }}})
   &ScaleColor(
    Scale Alpha=&FloatFromCurve(
     FloatCurve=DataInterface() {propertyValues {Curve {
        defaultValue=1
        keys [
         {arriveTangent=9.0725, interpMode=RCIM_Cubic, leaveTangent=9.0725, tangentMode=RCTM_User, time=0, value=0}
         {arriveTangent=-2.5249, interpMode=RCIM_Cubic, leaveTangent=-2.5249, tangentMode=RCTM_User, time=0.1408, value=0.9847}
         {arriveTangent=-0.0405, interpMode=RCIM_Cubic, leaveTangent=-0.0405, tangentMode=RCTM_User, time=0.5, value=0}
        ]
       }}}
     Scale Curve=&RandomRangeFloat(Minimum=0.7, Maximum=1)
    )
    Scale RGB=&MakeVectorFromLinearColorRGB(Color=&FNTimeOfDayColor(
      Night Color=<rgba 0.2842 0.3234 0.5503 1>
      Morning Color=<rgba 0.8455 0.6103 0.5753 1>
      Day Color=<rgba 1 1 1 1>
      Evening Color=<rgba 1 1 1 1>
     ))
    ScaleA=true
    ScaleRGB=true
   )
   /SetVariables(Particles.StarOutline_size:Vector2f=&Multiply_Vector2D_ByFloat(Float=1.2, Vector2D=$Particles.SpriteSize)) {isSetParameters=true}
  ]
  rendererClasses [&NiagaraMeshRendererProperties, &NiagaraSpriteRendererProperties]
  renderers [
   {rendererClass=&NiagaraMeshRendererProperties, rendererIndex=0}
   {rendererClass=&NiagaraMeshRendererProperties, rendererIndex=1}
   {rendererClass=&NiagaraSpriteRendererProperties, rendererIndex=2}
   {rendererClass=&NiagaraSpriteRendererProperties, rendererIndex=3}
  ]
 }
 TV_Speedlines {
  emitterUpdateScript [
   &EmitterState()
   &SpawnParticlesFromOtherEmitter(
    Attribute Reader=DataInterface() {propertyValues {EmitterBinding {bindingMode=Other, emitterName=TVs_and_Clouds}}}
    Spawn Rate=&FN_FrontEndFloatModifier(Gameplay=10, Front End=0)
    Spawn Rate Per Particle Cap=200
    Use Maximum Age=false
    Use Minimum Age=false
   )
  ]
  particleSpawnScript [
   &InitializeParticle(
    Color Mode=::NewEnumerator0
    Lifetime Max=1
    Lifetime Min=0.75
    Lifetime Mode=::NewEnumerator1
    Sprite Size=&Multiply_Vector2D_ByFloat(
     Float=&FN_FrontEndFloatModifier(Gameplay=1, Front End=0.4)
     Vector2D=&RandomRangeVector2D(Minimum=<15 100>, Maximum=<20 400>)
    )
    Sprite Size Mode=::NewEnumerator4
    UsePositionOffset=false
    Write Material Random=true
   )
   &SampleParticlesFromOtherEmitter(
    Age Sampling:ENiagara_DirectReadSamplingMode=::NewEnumerator0
    Lifetime Sampling:ENiagara_DirectReadSamplingMode=::NewEnumerator0
    Mesh Orientation Sampling:ENiagara_DirectReadSamplingMode=::NewEnumerator0
    Mesh Scale Sampling:ENiagara_DirectReadSamplingMode=::NewEnumerator0
    Particle ID Sampling=::NewEnumerator1
    Transform Data Option=1
    Velocity Application:ENiagara_DirectReadApplicationMode=::NewEnumerator0
    Velocity Sampling=::NewEnumerator0
   )
   &SpriteFacingAndAlignment(Sprite Alignment=<0 0 1>, WriteAlignment=true, WriteFacing=false)
  ]
  particleUpdateScript [
   &SolveForcesAndVelocity(Clamp Velocity=false, Limit Acceleration=false)
   &UpdateParticlesFromOtherEmitter(
    Age Sampling:ENiagara_DirectReadSamplingMode=::NewEnumerator0
    Attribute Reader=DataInterface() {propertyValues {EmitterBinding {bindingMode=Other, emitterName=TVs_and_Clouds}}}
    Lifetime Sampling:ENiagara_DirectReadSamplingMode=::NewEnumerator0
    Mesh Orientation Sampling:ENiagara_DirectReadSamplingMode=::NewEnumerator0
    Mesh Scale Sampling:ENiagara_DirectReadSamplingMode=::NewEnumerator0
    Transform Data Option=1
   )
   &ShapeLocation(
    Additional Quaternion Rotation=false
    Box Midpoint=<0.5 0.5 -0.15>
    Box Size=<80 50 70>
    MaskSpawnGroup=false
    Offset=<0 0 -40>
    Offset Mode=::NewEnumerator0
    Override Seed=false
    Rotation Mode=::NewEnumerator3
    Rotation Quaternion=$Output.UpdateParticlesFromOtherEmitter.SampledMeshOrientation
    Scale Mode:ENiagara_ScaleMode=::NewEnumerator1
    Shape Origin=$Output.UpdateParticlesFromOtherEmitter.SampledPosition
    Shape Primitive=::NewEnumerator2
    Surface Only=false
   )
   &ParticleState()
   &DynamicMaterialParameters(
    Index 0 Param 1=&RandomRangeFloat(Minimum=-0.3, Maximum=0.3)
    Index 0 Param 2=&RandomRangeFloat(Minimum=-0.3, Maximum=0.3)
    Module.Index0Float0Write=true
    Module.Index0Float1Write=true
    Module.Index0Float2Write=false
    Module.Index0Float3Write=false
   )
   /SetVariables(
    Particles.SpriteOutline:LinearColor=&MultiplyLinearColor(Linear Color=$Particles.Color, Scale Factor=&MakeVector4_FromLinearColor(LinearColor=<rgba 0 0 0 1>))
    Particles.SpriteOutlineWidth:Vector2f=&Multiply_Vector2D(A=$Particles.SpriteSize, B=&Multiply_Vector2D_ByFloat(Float=1.1, Vector2D=<1 1>))
   ) {isSetParameters=true}
   &CameraOffset(Camera Offset Amount=&RandomRangeFloat(Minimum=30, Maximum=40))
   &ScaleColor(
    Scale Alpha=&FloatFromCurve(FloatCurve=DataInterface() {propertyValues {Curve {
        defaultValue=1
        keys [
         {tangentMode=RCTM_User, time=0, value=1}
         {arriveTangent=-0.0405, interpMode=RCIM_Cubic, leaveTangent=-0.0405, tangentMode=RCTM_User, time=0.4975, value=1}
         {arriveTangent=-8.0089, interpMode=RCIM_Cubic, leaveTangent=-8.0089, tangentMode=RCTM_User, time=0.75, value=0}
        ]
       }}})
    ScaleA=true
    ScaleRGB=false
   )
  ]
  rendererClasses [&NiagaraSpriteRendererProperties]
  renderers [
   {rendererClass=&NiagaraSpriteRendererProperties}
   {rendererClass=&NiagaraSpriteRendererProperties, rendererIndex=1}
  ]
 }
 TVs_and_Clouds {
  emitterUpdateScript [
   &EmitterState()
   &SpawnRate(
    Spawn Probability=0.6
    SpawnRate=&FN_FrontEndFloatModifier(Gameplay=4.5, Front End=2.5)
    Use Spawn Probability=true
   )
  ]
  particleSpawnScript [
   /SetVariables(Particles.VisibilityTag=&SpawnGroup()) {isSetParameters=true}
   &InitializeParticle(
    Lifetime Max=2.4
    Lifetime Min=2
    Lifetime Mode=::NewEnumerator1
    Mesh Scale Mode=::NewEnumerator3
    Mesh Uniform Scale=&FN_FrontEndFloatModifier(Gameplay=1, Front End=0.45)
    Sprite Size Mode=::NewEnumerator1
    UsePositionOffset=false
    Write Material Random=true
   )
   &InitialMeshOrientation(
    Mesh Axis To Orient=<0 0 1>
    Mesh Orientation Mode=::NewEnumerator2
    Orientation Vector=<0 0 -1>
    Rotation=&RandomRangeVector(Minimum=<0 0 0>, Maximum=<0.25 0.25 0.25>)
    Use Rotation Vector=true
   )
   &ShapeLocation(
    Cylinder Distribution=::NewEnumerator2
    Cylinder Height=&FN_FrontEndFloatModifier(Gameplay=120, Front End=20)
    Cylinder Radius=&FN_FrontEndFloatModifier(Gameplay=250, Front End=110)
    MaskSpawnGroup=false
    Offset=&FN_FrontEndVector3Modifier(Gameplay=<0 0 550>, Front End=<0 45 120>)
    Offset Coordinate Space=::Local
    Offset Mode=::NewEnumerator0
    Rotation Axis=$System.FN_Contrail_Setup.Direction Vector
    Rotation Mode=::NewEnumerator1
    Shape Primitive=::NewEnumerator1
    Transform Order:ENiagara_TransformOrder=::NewEnumerator1
    U Position=&RandomRangeFloat(Minimum=&FN_FrontEndFloatModifier(Gameplay=0, Front End=0), Maximum=1)
    UseOwnerScale=true
    V Position=0
   )
  ]
  particleUpdateScript [
   &ParticleState()
   &ScaleMeshSize(Scale Factor=&VectorFromCurve(VectorCurve=DataInterface() {propertyValues {
       XCurve {keys [
         {arriveTangent=4.5495, interpMode=RCIM_Cubic, leaveTangent=4.5495, tangentMode=RCTM_User, time=0, value=0}
         {arriveTangent=1.5791, interpMode=RCIM_Cubic, leaveTangent=1.5791, time=0.15, value=1}
         {arriveTangent=-1.4286, interpMode=RCIM_Cubic, leaveTangent=-7.9375, tangentMode=RCTM_Break, time=0.6333, value=1}
         <at 0.8 0 RCIM_Cubic>
        ]}
       YCurve {keys [
         {arriveTangent=4.5495, interpMode=RCIM_Cubic, leaveTangent=4.5495, tangentMode=RCTM_User, time=0, value=0}
         {arriveTangent=1.5791, interpMode=RCIM_Cubic, leaveTangent=1.5791, time=0.15, value=1}
         {arriveTangent=-1.4286, interpMode=RCIM_Cubic, leaveTangent=-7.9375, tangentMode=RCTM_Break, time=0.6333, value=1}
         <at 0.8 0 RCIM_Cubic>
        ]}
       ZCurve {keys [
         {arriveTangent=4.5495, interpMode=RCIM_Cubic, leaveTangent=4.5495, tangentMode=RCTM_User, time=0, value=0}
         {arriveTangent=1.5791, interpMode=RCIM_Cubic, leaveTangent=1.5791, time=0.15, value=1}
         {arriveTangent=-1.4286, interpMode=RCIM_Cubic, leaveTangent=-7.9375, tangentMode=RCTM_Break, time=0.6333, value=1}
         <at 0.8 0 RCIM_Cubic>
        ]}
      }}))
   &ScaleColor(
    Scale Alpha=&FloatFromCurve(
     FloatCurve=DataInterface() {propertyValues {Curve {keys [
         {interpMode=RCIM_Cubic, tangentMode=RCTM_User, time=0, value=0}
         {interpMode=RCIM_Cubic, tangentMode=RCTM_User, time=0.5, value=1}
         {tangentMode=RCTM_User, time=1, value=0}
        ]}}}
     Scale Curve=$System.FN_Contrail_Setup.Normalized Intro Age
    )
    ScaleA=true
    ScaleRGB=true
   )
   /SetVariables(
    Particles.CloudOrientation:Quat4f=&MakeQuaternion(
     Quaternion From=1
     Axis=<1 1 1>
     Angle=&RandomRangeFloat(Minimum=0, Maximum=1)
     Yaw=&RandomRangeFloat(Minimum=0, Maximum=1)
     Pitch=&RandomRangeFloat(Minimum=0, Maximum=1)
     Roll=&RandomRangeFloat(Minimum=0, Maximum=1)
     From=?
     To=?
     Angle Type=::NewEnumerator1
     Normalize Vectors=false
    )
    Particles.CloudPosition:NiagaraPosition=&Lerp_Vector(
     A=&AddVectorToPosition(Position=$Particles.Position, Vector=<0 0 0>)
     B=&AddVectorToPosition(Position=$Particles.Position, Vector=<0 0 0>)
     Alpha=&FloatFromCurve(FloatCurve=DataInterface() {propertyValues {Curve {
         defaultValue=1
         keys [
          {arriveTangent=18.1967, interpMode=RCIM_Cubic, leaveTangent=18.1967, tangentMode=RCTM_User, time=-0.0064, value=0}
          {arriveTangent=2.8208, leaveTangent=2.8208, time=0.0621, value=1}
          {arriveTangent=-2.2834, interpMode=RCIM_Cubic, leaveTangent=-2.2834, time=0.35, value=1}
          {arriveTangent=-3.0592, leaveTangent=-3.0592, time=0.5, value=0}
          {arriveTangent=-0.316, interpMode=RCIM_Cubic, leaveTangent=-0.316, tangentMode=RCTM_User, time=0.8, value=0}
          {arriveTangent=1.8593, tangentMode=RCTM_User, time=1, value=1}
         ]
        }}})
    )
    Particles.CloudScale:Vector3f=&VectorFromCurve(
     VectorCurve=DataInterface() {propertyValues {
       XCurve {keys [
         {arriveTangent=34.4414, interpMode=RCIM_Cubic, leaveTangent=34.4414, tangentMode=RCTM_User, time=0, value=0}
         {arriveTangent=5.0132, interpMode=RCIM_Cubic, leaveTangent=5.0132, time=0.0834, value=1.9435}
         {arriveTangent=-4.6649, interpMode=RCIM_Cubic, leaveTangent=-4.6649, time=0.3534, value=1.7719}
         {arriveTangent=-11.0402, tangentMode=RCTM_User, time=0.5, value=0}
         {arriveTangent=29.211, interpMode=RCIM_Cubic, leaveTangent=29.211, tangentMode=RCTM_User, time=0.6, value=0}
         {arriveTangent=4.7801, interpMode=RCIM_Cubic, leaveTangent=4.7801, time=0.6532, value=1.6833}
         {arriveTangent=-4.8535, interpMode=RCIM_Cubic, leaveTangent=-4.8535, time=0.9107, value=1.4852}
         {arriveTangent=-19.4385, interpMode=RCIM_Cubic, leaveTangent=-19.4385, tangentMode=RCTM_User, time=1, value=0}
        ]}
       YCurve {keys [
         {arriveTangent=34.4414, interpMode=RCIM_Cubic, leaveTangent=34.4414, tangentMode=RCTM_User, time=0, value=0}
         {arriveTangent=5.0132, interpMode=RCIM_Cubic, leaveTangent=5.0132, time=0.0834, value=1.9435}
         {arriveTangent=-4.6649, interpMode=RCIM_Cubic, leaveTangent=-4.6649, time=0.3534, value=1.7719}
         {arriveTangent=-14.228, leaveTangent=-3.8596, tangentMode=RCTM_User, time=0.5, value=0}
         {arriveTangent=29.211, interpMode=RCIM_Cubic, leaveTangent=29.211, tangentMode=RCTM_User, time=0.6, value=0}
         {arriveTangent=4.7801, interpMode=RCIM_Cubic, leaveTangent=4.7801, time=0.6532, value=1.6833}
         {arriveTangent=-4.8535, interpMode=RCIM_Cubic, leaveTangent=-4.8535, time=0.9107, value=1.4852}
         {arriveTangent=-19.4385, interpMode=RCIM_Cubic, leaveTangent=-19.4385, tangentMode=RCTM_User, time=1, value=0}
        ]}
       ZCurve {keys [
         {arriveTangent=34.4414, interpMode=RCIM_Cubic, leaveTangent=34.4414, tangentMode=RCTM_User, time=0, value=0}
         {arriveTangent=5.0132, interpMode=RCIM_Cubic, leaveTangent=5.0132, time=0.0834, value=1.9435}
         {arriveTangent=-4.6649, interpMode=RCIM_Cubic, leaveTangent=-4.6649, time=0.3534, value=1.7719}
         {arriveTangent=-13.7325, tangentMode=RCTM_User, time=0.5, value=0}
         {arriveTangent=26.992, interpMode=RCIM_Cubic, leaveTangent=26.992, tangentMode=RCTM_User, time=0.6, value=0}
         {arriveTangent=4.7801, interpMode=RCIM_Cubic, leaveTangent=4.7801, time=0.6532, value=1.6833}
         {arriveTangent=-4.8535, interpMode=RCIM_Cubic, leaveTangent=-4.8535, time=0.9107, value=1.4852}
         {arriveTangent=-19.4385, tangentMode=RCTM_User, time=1, value=0}
        ]}
      }}
     Scale Curve=&Multiply_Vector(
      A=&VectorFromFloat(Value=&RandomRangeFloat(Minimum=1.3, Maximum=1.5))
      B=&FN_FrontEndVector3Modifier(Gameplay=<1 1 1>, Front End=<0.4 0.4 0.4>)
     )
    )
    Particles.Cloud_DMP:Vector4f=&MakeVector4(
     X=&FloatFromCurve(FloatCurve=DataInterface() {propertyValues {Curve {
         defaultValue=0
         keys [
          {arriveTangent=-0.9265, interpMode=RCIM_Cubic, leaveTangent=-0.9265, tangentMode=RCTM_User, time=0, value=0}
          {arriveTangent=3.3137, leaveTangent=3.3137, time=0.25, value=1}
          {interpMode=RCIM_Constant, tangentMode=RCTM_User, time=0.4, value=1}
          <at 0.5 0>
          {interpMode=RCIM_Cubic, tangentMode=RCTM_User, time=0.6, value=0}
          {arriveTangent=0.8927, interpMode=RCIM_Cubic, leaveTangent=0.8927, tangentMode=RCTM_User, time=1, value=1}
         ]
        }}})
     Y=1
     Z=1
     W=1
    )
    Particles.Color=&FNTimeOfDayColor(
     Night Color=<rgba 0.2842 0.3234 0.5503 1>
     Morning Color=<rgba 0.8455 0.6103 0.5753 1>
     Day Color=<rgba 1 1 1 1>
     Evening Color=<rgba 1 1 1 1>
    )
   ) {isSetParameters=true}
   &MeshRotationForce(Rotation=&RandomRangeVector(
     Minimum=&RandomRangeVector(Minimum=<-0.5 -0.5 -0.5>, Maximum=<-2 -2 -2>)
     Maximum=&RandomRangeVector(Minimum=<-2.5 -2.5 -2.5>, Maximum=<3 3 3>)
    ))
   &KillParticlesInVolume(Origin Offset:Vector3f=<0 -100 -500>, Sphere Radius=200)
   &GravityForce(Gravity=&AddVectorToPosition(
     Position=&FN_FrontEndVector3Modifier(Gameplay=<0 0 980>, Front End=<0 0 600>)
     Vector=$System.FN_Contrail_Setup.Direction Vector
    ))
   &GravityForce(Gravity=&FN_FrontEndVector3Modifier(
     Gameplay=<0 0 980>
     Front End=&TransformVector(Vector=<0 -200 600>, Source Space=::World, Destination Space=::Local)
    )) {enabled=false, module=GravityForce001}
   &SolveForcesAndVelocity(Clamp Velocity=false, Limit Acceleration=false)
   &DynamicMaterialParameters(
    Index 2 Param 2=&SelectFloatFromArray(
     Array Sampling Mode=::NewEnumerator0
     Direct Array Index=?
     Float Selection Array=DataInterface() {propertyValues {FloatData [0,0.3,0.65,1]}}
     Random Seed=0
    )
    Index 3 Param 1=&RandomRangeFloat(Minimum=-1, Maximum=-1.5)
    Index 3 Param 2=&RandomRangeFloat(Minimum=7, Maximum=25)
    Index 3 Param 3=&RandomRangeFloat(Minimum=-12, Maximum=12)
    Index 3 Param 4=&RandomRangeFloat(Minimum=0.5, Maximum=1.3)
    Module.Index2Float0Write=false
    Module.Index2Float1Write=true
    Module.Index2Float2Write=true
    Module.Index2Float3Write=true
    Module.Index3Float0Write=true
    Module.Index3Float1Write=true
    Module.Index3Float2Write=true
    Module.Index3Float3Write=true
    Write Parameter Index 0=::NewEnumerator0
    Write Parameter Index 2=::NewEnumerator1
    Write Parameter Index 3=::NewEnumerator1
   )
   /SetVariables(Particles.CordScale:Vector3f=&MakeCustomVectorFromBool(
     Bool=&SetBoolByFloatComparison(
      A=&MakeFloatFromVector4(Vector 4=<0 0 -15 0>, Channel=::NewEnumerator2)
      B=&MakeFloatFromVector4(Vector 4=$Particles.DynamicMaterialParameter3, Channel=::NewEnumerator2)
      Comparison Type=::NewEnumerator7
     )
     True Vector=<1 0.75 1>
     False Vector=&MakeCustomVectorFromBool(
      Bool=&SetBoolByFloatComparison(
       A=&MakeFloatFromVector4(Vector 4=<0 0 15 0>, Channel=::NewEnumerator2)
       B=&MakeFloatFromVector4(Vector 4=$Particles.DynamicMaterialParameter3, Channel=::NewEnumerator2)
       Comparison Type=::NewEnumerator2
      )
      True Vector=<1 0.75 1>
      False Vector=<1 0.85 1>
     )
    )) {isSetParameters=true}
   /SetVariables(Particles.CordScale:Vector3f=&Multiply_Vector(
     A=$Particles.CordScale
     B=&VectorFromFloat(Value=&FloatFromCurve(
       FloatCurve=DataInterface() {propertyValues {Curve {
          defaultValue=1
          keys [
           {arriveTangent=6.8229, interpMode=RCIM_Cubic, leaveTangent=6.8229, tangentMode=RCTM_User, time=0.05, value=0}
           <at 0.5 1>
           {arriveTangent=-5, interpMode=RCIM_Cubic, leaveTangent=-5, time=0.6, value=1}
           {arriveTangent=-20.2891, tangentMode=RCTM_User, time=0.7, value=0}
          ]
         }}}
       Scale Curve=&FN_FrontEndFloatModifier(Gameplay=1, Front End=0.5)
      ))
    )) {isSetParameters=true}
   /SetVariables(Particles.TVDynamicParam:Vector4f=&MakeVector4(
     X=&UniformAOrBFloat(A=0, B=0.5, Evaluation Type=::NewEnumerator0, Recalculate Random Each Loop=true, Override Seed=false)
     Y=&UniformAOrBFloat(A=0, B=0.5, Evaluation Type=::NewEnumerator0, Recalculate Random Each Loop=true, Override Seed=false)
     Z=&FNTimeOfDayScalar(Night Scale=0.4, Morning Scale=0, Day Scale=0, Evening Scale=0)
     W=&UniformAOrBFloat(A=0, Distribution Weight=0.25, Evaluation Type=::NewEnumerator0, Recalculate Random Each Loop=true, Override Seed=false)
    )) {isSetParameters=true}
  ]
  rendererClasses [&NiagaraMeshRendererProperties]
  renderers [
   {rendererClass=&NiagaraMeshRendererProperties, rendererIndex=0}
   {rendererClass=&NiagaraMeshRendererProperties, rendererIndex=1}
   {rendererClass=&NiagaraMeshRendererProperties, rendererIndex=2}
   {rendererClass=&NiagaraMeshRendererProperties, rendererIndex=3}
   {rendererClass=&NiagaraMeshRendererProperties, rendererIndex=4}
   {rendererClass=&NiagaraMeshRendererProperties, rendererIndex=5}
  ]
 }
}
summary {
 systemName=NS_Contrail_FruitFire_Nudge
 emitters [
  {emitterName=TVs_and_Clouds, rendererClasses [&NiagaraMeshRendererProperties]}
  {emitterName=Ambient_Speedlines, rendererClasses [&NiagaraSpriteRendererProperties]}
  {emitterName=TV_Speedlines, rendererClasses [&NiagaraSpriteRendererProperties]}
  {emitterName=SpawningClouds, rendererClasses [&NiagaraMeshRendererProperties, &NiagaraSpriteRendererProperties]}
 ]
 userVariables [
  User.bIsFrontEnd {type=NiagaraBool}
  User.bIsFrontEndPreview {type=NiagaraBool}
  User.ParaGlideLeanAlpha {type=NiagaraFloat, defaultValue=0, description=""}
  User.RibbonIDOffset {type=NiagaraInt32}
 ]
}
